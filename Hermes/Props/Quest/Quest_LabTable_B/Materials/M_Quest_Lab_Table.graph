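# Quest lab table prop material that adds animated bubbling lab liquid to the Base4Layers GigaMix output wherever its emissive is non-zero.
# Camera-offset, actor-relative world projections pan bubble textures, with separate top and side treatment, fresnel and scrolling dual normals.
# The tinted result drives base color and exposure-compensated emissive.
# Shared damage, WPO, snow, seasonal and RVT functions follow, with flattened specular and roughness on low-quality mobile.
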Material /Hermes/Props/Quest/Quest_LabTable_B/Materials/M_Quest_Lab_Table
graph {
 properties {
  bAutomaticallySetUsageInEditor=false
  bTangentSpaceNormal=false
  bUseMaterialAttributes=true
  bUsedWithInstancedStaticMeshes=true
  bUsedWithNanite=true
  bUsedWithSplineMeshes=true
  ditheredLODTransition=true
  pixelDepthOffsetMode=PDOM_Legacy
  shadingModel=MSM_FromMaterialExpression
  usedShadingModels=Default Lit | Subsurface | Clear Coat | Two Sided Foliage
 }
 nodes [
  n1 = Add(
   A=Multiply(A=CameraVectorWS()) {constB=-10}
   B=Subtract(A=WorldPosition() {worldPositionShaderOffset=WPT_CameraRelative}.XYZ, B=ActorPositionWS() {originType=CameraRelative})
  )
  n2 = Add(
   A=Multiply(A=@n1, B=Constant3Vector() {constant=<rgba 0.048 0.051 0.049 0>}) {constB=0.05}
   B=Multiply(A=Time(), B=Constant3Vector() {constant=<rgba 0.02 -0.01 -0.3 0>}) {constB=0.5}
  )
  n3 = ComponentMask(None=@n2) {r=true}
  n4 = Multiply(A=ComponentMask(None=@n2) {b=true}) {constB=-1}
  n5 = ComponentMask(None=@n1) {g=true, r=true}
  n6 = Multiply(A=@n5) {constB=0.017}
  n7 = Time()
  n8 = Multiply(A=@n7, B=Constant2Vector() {g=0.017, r=0.02})
  n9 = TextureSample(UVs=LinearInterpolate(A=AppendVector(A=@n3, B=@n4), B=Add(A=@n6, B=@n8), Alpha=NamedRerouteUsage())) {samplerType=SAMPLERTYPE_LinearColor, texture=&/T_LabLiquid_Bubbles}
  n10 = Add(A=@n9.G) {constB=-0.5}
  n11 = Multiply(A=Time()) {constB=2}
  n12 = Add(A=@n10, B=Add(A=@n9.B, B=@n11))
  n13 = Abs(None=Add(A=Frac(None=@n12)) {constB=-0.5})
  Add(A=@n13) {constB=-0.2}
  n14 = ComponentMask(None=@n2) {g=true}
  n15 = Multiply(A=@n5) {constB=-0.02}
  n16 = Multiply(A=@n7, B=Constant2Vector() {g=-0.017, r=0.02})
  n17 = TextureSample(UVs=LinearInterpolate(A=AppendVector(A=@n14, B=@n4), B=Add(A=@n15, B=@n16), Alpha=NamedRerouteUsage())) {samplerType=SAMPLERTYPE_LinearColor, texture=&T_LabLiquid_Bubbles}
  n18 = Add(A=@n17.G) {constB=-0.5}
  n19 = Add(A=@n18, B=Add(A=@n17.B, B=@n11))
  n20 = Abs(None=Add(A=Frac(None=@n19)) {constB=-0.5})
  Add(A=@n20) {constB=-0.2}
  n21 = GetMaterialAttributes(None=MaterialFunctionCall() {materialFunction=/Game/Global/Base4Layers_2022/MaterialFunctions/MF_GigaMix_2023}.Result) {attributeGetTypes [#1, #2, #3]}
  n22 = Transform(None=PixelNormalWS()) {transformSourceType=TRANSFORMSOURCE_World, transformType=TRANSFORM_View}
  n23 = ComponentMask(None=Add(A=Multiply(A=@n22) {constB=0.5}) {constB=0.5}) {g=true, r=true}
  n24 = Transform(None=Multiply(A=NamedRerouteUsage(), B=Constant3Vector() {constant=<rgba -1 -1 1 0>}))
  n25 = CameraVectorWS()
  n26 = OneMinus(None=Abs(None=DotProduct(A=@n24, B=@n25)))
  n27 = Multiply(A=@n26, B=@n26)
  n28 = Abs(None=ComponentMask(None=PixelNormalWS()) {g=true, r=true})
  n29 = Multiply(A=@n28, B=@n28)
  n30 = Divide(A=@n29, B=DotProduct(A=@n29, B=Constant2Vector() {g=1, r=1}))
  n31 = ComponentMask(None=@n30) {g=true}
  n32 = ComponentMask(None=@n30) {r=true}
  n33 = Add(A=Add(A=Multiply(A=@n9.G, B=@n31), B=Multiply(A=@n17.G, B=@n32))) {constB=-0.5}
  n34 = Add(
   A=LinearInterpolate(
    A=TextureSample(UVs=@n23) {samplerType=SAMPLERTYPE_LinearColor, texture=&T_LabLiquid_Bubbles}.R
    B=Multiply(A=@n27) {constB=1.5}
    Alpha=NamedRerouteUsage()
   )
   B=LinearInterpolate(A=Multiply(A=@n33) {constB=4}, B=NamedRerouteUsage(), Alpha=NamedRerouteUsage())
  )
  n35 = Multiply(
   A=Desaturation(
    None=Multiply(A=@n21.Base Color, B=@n34)
    Fraction=ScalarParameter() {defaultValue=-0.25, group=LabLiquid, parameterName=LabLiquid_Desaturation}
   )
   B=ScalarParameter() {defaultValue=2, group=LabLiquid, parameterName=LabLiquid_Multiplier}
  ) {constB=2}
  n36 = Constant() {r=0.7}
  n37 = NamedRerouteUsage()
  n38 = NamedRerouteUsage()
  n39 = BlendMaterialAttributes(
   A=@n21.MaterialAttributes
   B=SetMaterialAttributes!(
    MaterialAttributes=@n21.MaterialAttributes
    Base Color=@n35
    Emissive Color=EyeAdaptationInverse(LightValueInput=@n35, AlphaInput=@n36).EyeAdaptationInverse
    Normal=LinearInterpolate(A=@n21.Normal, B=@n37, Alpha=@n38)
   ) {attributeSetTypes [#1, #2, #3]}
   Alpha=Ceil(None=Add(A=@n21.Emissive Color) {constB=-0.001})
  ) {vertexAttributeBlendType=UseA}
  MaterialExpressionReroute_70 = Reroute(None=@n39)
  n40 = ScalarParameter() {
   desc=Documentation link
   group="0 (Expand)  Documentation (RMB - Copy display name) and Debug view"
   parameterName=https://confluence.it.epicgames.com/display/FORT/Fortnite+Base+Material+2022
  }
  n41 = SetMaterialAttributes!(MaterialAttributes=@MaterialExpressionReroute_70.MaterialAttributes, World Position Offset=Multiply(A=@n40)) {attributeSetTypes [#4]}
  MaterialExpressionReroute_71 = Reroute(None=@n41)
  MaterialExpressionReroute_69 = Reroute(None=@MaterialExpressionReroute_71.Output)
  n42 = MaterialFunctionCall(Material_In=@MaterialExpressionReroute_69.MaterialAttributes) {materialFunction=&MF_FN_DestructibleDamage}
  n43 = MaterialFunctionCall!(
   In=MaterialFunctionCall!(Material Attributes=MaterialFunctionCall!(In=@n42.Material_Out) {materialFunction=/Game/Athena/WinterArt/Materials/MF_WPO_WindNew}.Result) {materialFunction=/Game/Creative/Sets/MilitaryBase/Props/MF_FanSpin}.Result
   UseSphericalMappingOverride
  ) {materialFunction=/Game/Packages/Fortress_SharedMaterials/Base_Material_Functions/MF_CurieFallbackEffect}
  n44 = MaterialFunctionCall!(In=MaterialFunctionCall!(In=MaterialFunctionCall!(In=@n43.Result, Custom Mask, Custom Rotation Axis) {materialFunction=/Game/Athena/WinterArt/Materials/MF_WindWPO_v2}.Result) {materialFunction=/Game/Global/Base4Layers_2022/MaterialFunctions/MF_SnowAndIceArchitecture}.Result) {materialFunction=/Game/Global/Base4Layers_2022/MaterialFunctions/MF_VibrationRattle}
  MaterialExpressionReroute_68 = Reroute(None=@n44.Result)
  Composite!(None=@MaterialExpressionReroute_68.Result) {desc=Transform to World in case we will use RVT after this point, subgraphName=Tangent to World}
  Composite!(MaterialAttributes=@n39) {subgraphName=Documentation Link}
  Composite(MaterialAttributes=@MaterialExpressionReroute_71.Output) {desc=Game Shared WPO and Fire Effect, subgraphName=WPO And Shared Functions}
  MaterialExpressionReroute_1 = Reroute(None=@n51.RVT H)
  MaterialExpressionReroute_3 = Reroute(None=@MaterialExpressionReroute_68.Result)
  n45 = GetMaterialAttributes(None=@MaterialExpressionReroute_3.Input) {attributeGetTypes [#3]}
  n46 = Transform(None=@n45.Normal)
  n47 = LinearInterpolate(A=Multiply(A=@n46) {constB=-1}, B=@n46, Alpha=Saturate(None=TwoSidedSign())) {desc="Fixes normals\r\non backfaces"}
  n48 = SetMaterialAttributes!(
   MaterialAttributes=@n45.MaterialAttributes
   Normal=StaticSwitchParameter(True=MaterialProxyReplace(Realtime=@n47, MaterialProxy=@n45.Normal), False=@n47) {parameterName=Force TangentSpaceNormalsFor HLOD ProxyReplace}
  ) {attributeSetTypes [#3]}
  MaterialExpressionReroute_2 = Reroute(None=@n48)
  n49 = MaterialFunctionCall!(In=MaterialFunctionCall!(In=@MaterialExpressionReroute_2.Output) {materialFunction=/Game/ArtTools/Tech_Art/DirectFX/Assets/MaterialFunctions/MF_DirectFX}.Result) {materialFunction=/Game/ArtTools/Tech_Art/Veneer/MaterialFunctions/MF_Veneer}
  MaterialExpressionReroute_8 = Reroute(None=@n49.Result)
  n50 = MaterialFunctionCall!(AO, RVT_Height=@MaterialExpressionReroute_1."RVT_Height (S)", For Buildings=@MaterialExpressionReroute_8."For Moss (MA)", For Moss) {materialFunction=/Game/Environments/Asteria/Rocks/Materials/MaterialFunctions/MF_JungleBiome}
  MaterialExpressionReroute_0 = Reroute(None=@n50.Buildings Jungl)
  n51 = MaterialFunctionCall!(
   In=@MaterialExpressionReroute_0.Result
   "RVT Height Map (Roughness)"=GetMaterialAttributes(None=@n49.Result) {attributeGetTypes [#5]}.Roughness
  ) {materialFunction=/Game/Global/Base4Layers_2022/MaterialFunctions/MF_RVTBlending}
  Composite!(InputPin=@n51.Result) {subgraphName=Mobile Opt}
  Composite!("For Moss (MA)"=@n49.Result, "RVT_Height (S)"=@n51.RVT H) {subgraphName="Seasonal "}
  MaterialExpressionReroute_7 = Reroute(None=@n51.Result)
  n52 = GetMaterialAttributes(None=@MaterialExpressionReroute_7.InputPin)
  n53 = Constant() {r=0.5}
  n54 = Constant() {r=0.8}
  n55 = ShadingPathSwitch(
   Default=@MaterialExpressionReroute_7.InputPin
   Mobile=QualitySwitch(
    Default=@MaterialExpressionReroute_7.InputPin
    Low=SetMaterialAttributes!(MaterialAttributes=@n52.MaterialAttributes, Specular=@n53, Roughness=@n54) {attributeSetTypes [#6, #5]}
    High=@MaterialExpressionReroute_7.InputPin
    Medium=@MaterialExpressionReroute_7.InputPin
    Epic=@MaterialExpressionReroute_7.InputPin
   )
  )
  MaterialExpressionReroute_6 = Reroute(None=@n55)
  MaterialFunctionCall(MA In=@MaterialExpressionReroute_6.Output) {materialFunction=&MF_CreativeMoveTool}
  n56 = Multiply(A=CameraVectorWS()) {constB=-1}
  n57 = Constant() {r=0.1}
  Multiply(A=@n21.Base Color, B=Multiply(A=SkyLightEnvMapSample(Direction=@n56, Roughness=@n57)) {constB=3})
  n58 = VertexNormalWS()
  n59 = Add(A=ComponentMask(None=Abs(None=@n58)) {b=true}) {constB=0.01}
  NamedRerouteDeclaration(None=Saturate(None=Floor(None=@n59))) {name=Z Mask, nodeColor=<rgba 0 1 0.5078 1>}
  n60 = Add(A=Multiply(A=@n13, B=Abs(None=@n10)), B=Multiply(A=@n20, B=Abs(None=@n18)))
  n61 = Multiply(A=Multiply(A=@n60, B=@n60)) {constB=20}
  NamedRerouteDeclaration(None=@n61) {name=Top Bubbles, nodeColor=<rgba 0.3594 0 1 1>}
  n62 = Multiply(A=TextureCoordinate(), B=ScalarParameter() {defaultValue=1, group=LabLiquid, parameterName=LabLiquid_NormalTiling})
  n63 = Time()
  n64 = Constant2Vector() {g=0.17, r=0.2}
  n65 = TextureSample(UVs=Add(A=Multiply(A=@n62) {constB=12}, B=Multiply(A=@n63, B=@n64))) {samplerType=SAMPLERTYPE_Normal, texture=&/T_LabLiquid_Bubbles_N}
  n66 = Constant2Vector() {g=-0.17, r=-0.2}
  n67 = TextureSample(UVs=Add(A=Multiply(A=@n62) {constB=18}, B=Multiply(A=@n63, B=@n66))) {samplerType=SAMPLERTYPE_Normal, texture=&T_LabLiquid_Bubbles_N}
  n68 = AppendVector(A=Add(A=ComponentMask(None=@n65.RGB) {g=true, r=true}, B=ComponentMask(None=@n67.RGB) {g=true, r=true}), B=Multiply(A=@n65.B, B=@n67.B))
  n69 = Constant3Vector() {constant=<rgba -1 -1 1 0>}
  n70 = Multiply(A=NamedRerouteUsage()) {constB=10}
  NamedRerouteDeclaration(None=LinearInterpolate(A=@n68, B=Multiply(A=@n68, B=@n69) {constB=-1}, Alpha=Saturate(None=@n70))) {name=Normal Top, nodeColor=<rgba 0.9688 0 1 1>}
  PinBase() {reroutePins [Input {expression=&self:MaterialExpressionReroute_3}]}
  PinBase!(Output=@n48) {reroutePins [Output {expression=&self:MaterialExpressionReroute_2}]}
  PinBase!() {reroutePins [
    "For Moss (MA)" {expression=&self:MaterialExpressionReroute_8}
    "RVT_Height (S)" {expression=&self:MaterialExpressionReroute_1}
   ]}
  PinBase!(Result=@n50.Buildings Jungl) {reroutePins [Result {expression=&self:MaterialExpressionReroute_0}]}
  PinBase!() {reroutePins [MaterialAttributes {expression=&self:MaterialExpressionReroute_70}]}
  PinBase!(Output=@n41) {reroutePins [Output {expression=&self:MaterialExpressionReroute_71}]}
  PinBase!() {reroutePins [MaterialAttributes {expression=&self:MaterialExpressionReroute_69}]}
  PinBase!(Result=@n44.Result) {reroutePins [Result {expression=&self:MaterialExpressionReroute_68}]}
  PinBase!() {reroutePins [InputPin {expression=&self:MaterialExpressionReroute_7}]}
  PinBase!(Output=@n55) {reroutePins [Output {expression=&self:MaterialExpressionReroute_6}]}
  Reroute(None=@n61)
  Reroute(None=@n5)
 ]
}
